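; Music encounter Blueprint for the HeavyLamp boss fight, driving its music through encounter state changes.
; Idle and Combat share one MetaSound, so entering either state calls the parent and sets the boss combat phase to 0 or 1.
; A "Defeated" named event switches the boss encounter state to an enumerated value.
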(event EventEnterIdleState
  (|CallParentFunction) ; Idle and Combat share the same metasound
  (CallFunction|SetBossEncounterCombatPhase 0))

(event EventEnterCombatState
  (|CallParentFunction)
  (CallFunction|SetBossEncounterCombatPhase 1))

(event MusicEncounter|EventExecuteNamedEvent
  (params (Music%20Encounter%20Component :type Music%20Encounter%20Component%20Base%20Object%20Reference) (Event%20Name :type Name :default "None"))
  (switch name Event%20Name
    (:Defeated
      (CallFunction|SetBossEncounterState "NewEnumerator4"))))
